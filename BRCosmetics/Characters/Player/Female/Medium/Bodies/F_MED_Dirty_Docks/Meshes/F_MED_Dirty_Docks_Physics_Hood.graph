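# Physics asset for the hood of the F_MED_Dirty_Docks female player body, defining its dynamic hood bones.
# Box bodies on dyn_hood_1_1 and dyn_hood_1_2 form a two-link chain from a small dyn_hood_offset box, with tight, asymmetric swing and twist limits at each joint.
# A capsule on the head bone has no constraint to the chain, apparently serving as a collision shape for the hood.
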PhysicsAsset /BRCosmetics/Characters/Player/Female/Medium/Bodies/F_MED_Dirty_Docks/Meshes/F_MED_Dirty_Docks_Physics_Hood
bodies {
 dyn_hood_1_1 [{shapeType=Box, extentX=1, extentY=7.75, extentZ=6.9375, center=<0.6343 -4.4073 0>, rotation=<rot 0 8.4395 0>}]
 dyn_hood_1_2 [{shapeType=Box, extentX=4.1875, extentY=6.875, extentZ=1, center=<2.1388 0 0.0007>}]
 dyn_hood_offset [{shapeType=Box, extentX=1, extentY=1, extentZ=1}]
 head [{shapeType=Capsule, length=9.0926, radius=3.8857, center=<-0.0162 0.5363 -0.0126>, rotation=<rot -0.0278 49.2373 -89.4327>}]
}
constraints [
 {bone1Name=dyn_hood_1_1, bone2Name=dyn_hood_offset, swing1LimitDegrees=25, swing2LimitDegrees=5, twistLimitDegrees=5}
 {bone1Name=dyn_hood_1_2, bone2Name=dyn_hood_1_1, swing1LimitDegrees=10, swing2LimitDegrees=20, twistLimitDegrees=5}
]
