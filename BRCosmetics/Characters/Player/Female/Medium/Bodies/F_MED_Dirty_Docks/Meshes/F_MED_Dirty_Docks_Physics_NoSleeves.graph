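# Physics asset for the NoSleeves variant of the female Dirty Docks body, covering secondary motion of jacket flaps, arm bands and a right sleeve.
# Seven jacket flaps hinge with limited swing and twist on small master bodies, which attach with free rotation to spine_03 in a capsule-based torso chain.
# Three arm bands per forearm hang off lowerarm and twist bodies; left forearm links are angle-limited, right ones free.
# Despite the name, a tightly limited dyn_sleeve_r body remains on upperarm_twist_01_r.
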PhysicsAsset /BRCosmetics/Characters/Player/Female/Medium/Bodies/F_MED_Dirty_Docks/Meshes/F_MED_Dirty_Docks_Physics_NoSleeves
bodies {
 Dyn_ArmBand_L1 [{shapeType=Box, extentX=1, extentY=9.4375, extentZ=5, center=<0 0 -2.9272>, rotation=<rot 11.0738 -8.5725 -2.2147>}]
 Dyn_ArmBand_L2 [{shapeType=Box, extentX=1, extentY=7.6875, extentZ=4.875, center=<0.3429 -0.0338 -2.7461>, rotation=<rot 5.6251 -5.626 0>}]
 Dyn_ArmBand_L3 [{shapeType=Box, extentX=1, extentY=6.0625, extentZ=5.1875, center=<0.3873 -0.0002 -2.7881>, rotation=<rot 14.0628 0 0>}]
 Dyn_ArmBand_R1 [{shapeType=Box, extentX=1, extentY=9.125, extentZ=4.8125, center=<0.2979 -0.0001 2.4106>, rotation=<rot -8.4377 0 -5.6252>}]
 Dyn_ArmBand_R2 [{shapeType=Box, extentX=1, extentY=7, extentZ=6, center=<0.3121 0.0003 2.7661>, rotation=<rot -8.4377 0 0>}]
 Dyn_ArmBand_R3 [{shapeType=Box, extentX=1, extentY=6.125, extentZ=4.625, center=<0 0 2.2927>, rotation=<rot -11.2502 0 22.4996>}]
 dyn_jacket_bk_01 [{shapeType=Box, extentX=6, extentY=9, extentZ=0.5, center=<2.7 0 0.5>}]
 dyn_jacket_bk_MSTR [{shapeType=Box, extentX=1, extentY=1, extentZ=1}]
 dyn_jacket_bk_l_01 [{shapeType=Box, extentX=3.5, extentY=7, extentZ=1, center=<1.5 0 0>, rotation=<rot 0 0 10>}]
 dyn_jacket_bk_l_MSTR [{shapeType=Box, extentX=1, extentY=1, extentZ=1}]
 dyn_jacket_bk_r_01 [{shapeType=Box, extentX=3.5, extentY=7, extentZ=0.8, center=<-1.5 0 0>, rotation=<rot 0 0 12>}]
 dyn_jacket_bk_r_MSTR [{shapeType=Box, extentX=1, extentY=1, extentZ=1}]
 dyn_jacket_fr_l_01 [{shapeType=Box, extentX=3, extentY=6, extentZ=0.5, center=<1.2 0 0>, rotation=<rot 0 0 -5>}]
 dyn_jacket_fr_l_MSTR [{shapeType=Box, extentX=1, extentY=1, extentZ=1}]
 dyn_jacket_fr_r_01 [{shapeType=Box, extentX=3, extentY=6, extentZ=1, center=<-1.3 0 0>}]
 dyn_jacket_fr_r_MSTR [{shapeType=Box, extentX=1, extentY=1, extentZ=1}]
 dyn_jacket_l_01 [{shapeType=Box, extentX=2.5, extentY=6.5, extentZ=1, center=<1.3 0 0>, rotation=<rot 0 0 10>}]
 dyn_jacket_l_MSTR [{shapeType=Box, extentX=1, extentY=1, extentZ=1}]
 dyn_jacket_r_01 [{shapeType=Box, extentX=2.5, extentY=7, extentZ=1, center=<-1 0 0>}]
 dyn_jacket_r_MSTR [{shapeType=Box, extentX=1, extentY=1, extentZ=1}]
 dyn_sleeve_r [{shapeType=Box, extentX=9.8, extentY=8.2, extentZ=9.3, center=<-4.3 0 0>, rotation=<rot 0 -89 0>}]
 lowerarm_l [{shapeType=Box, extentX=1, extentY=1, extentZ=1}]
 lowerarm_r [{shapeType=Box, extentX=1, extentY=1, extentZ=1}]
 lowerarm_twist_01_l [{shapeType=Box, extentX=1, extentY=1, extentZ=1}]
 lowerarm_twist_01_r [{shapeType=Box, extentX=1, extentY=1, extentZ=1}]
 lowerarm_twist_02_l [{shapeType=Box, extentX=1, extentY=1, extentZ=1}]
 lowerarm_twist_02_r [{shapeType=Box, extentX=1, extentY=1, extentZ=1}]
 spine_03 [{shapeType=Capsule, length=2, radius=8.5, center=<0 2.082 0>}]
 spine_04 [{shapeType=Capsule, length=3.5, radius=11, center=<0 2.5 0>}]
 spine_05 [{shapeType=Capsule, length=12.636, radius=11.8126, center=<-3.7799 0.0702 1.492>, rotation=<rot 1.8829 92.3552 -90>}]
 upperarm_twist_01_r [{shapeType=Box, extentX=14.95, extentY=23.26, extentZ=1.08, center=<0.12 -1.95 5.79>, rotation=<rot 0 0 -5.8>}]
}
constraints [
 {bone1Name=dyn_jacket_r_01, bone2Name=dyn_jacket_r_MSTR, swing1LimitDegrees=15, swing2LimitDegrees=45, twistLimitDegrees=10}
 {bone1Name=dyn_jacket_bk_r_01, bone2Name=dyn_jacket_bk_r_MSTR, swing1LimitDegrees=15, swing2LimitDegrees=45, twistLimitDegrees=10}
 {bone1Name=dyn_jacket_bk_01, bone2Name=dyn_jacket_bk_MSTR, swing1LimitDegrees=15, swing2LimitDegrees=45, twistLimitDegrees=10}
 {bone1Name=dyn_jacket_bk_l_01, bone2Name=dyn_jacket_bk_l_MSTR, swing1LimitDegrees=15, swing2LimitDegrees=45, twistLimitDegrees=10}
 {bone1Name=dyn_jacket_l_01, bone2Name=dyn_jacket_l_MSTR, swing1LimitDegrees=15, swing2LimitDegrees=45, twistLimitDegrees=10}
 {bone1Name=dyn_jacket_fr_r_01, bone2Name=dyn_jacket_fr_r_MSTR, swing1LimitDegrees=15, swing2LimitDegrees=45, twistLimitDegrees=10}
 {bone1Name=dyn_jacket_fr_l_01, bone2Name=dyn_jacket_fr_l_MSTR, swing1LimitDegrees=15, swing2LimitDegrees=45, twistLimitDegrees=10}
 {
  bone1Name=dyn_jacket_r_MSTR
  bone2Name=spine_03
  swing1LimitDegrees=45
  swing1Motion=Free
  swing2LimitDegrees=45
  swing2Motion=Free
  twistLimitDegrees=45
  twistMotion=Free
 }
 {
  bone1Name=dyn_jacket_bk_r_MSTR
  bone2Name=spine_03
  swing1LimitDegrees=45
  swing1Motion=Free
  swing2LimitDegrees=45
  swing2Motion=Free
  twistLimitDegrees=45
  twistMotion=Free
 }
 {
  bone1Name=dyn_jacket_bk_MSTR
  bone2Name=spine_03
  swing1LimitDegrees=45
  swing1Motion=Free
  swing2LimitDegrees=45
  swing2Motion=Free
  twistLimitDegrees=45
  twistMotion=Free
 }
 {
  bone1Name=dyn_jacket_bk_l_MSTR
  bone2Name=spine_03
  swing1LimitDegrees=45
  swing1Motion=Free
  swing2LimitDegrees=45
  swing2Motion=Free
  twistLimitDegrees=45
  twistMotion=Free
 }
 {
  bone1Name=dyn_jacket_l_MSTR
  bone2Name=spine_03
  swing1LimitDegrees=45
  swing1Motion=Free
  swing2LimitDegrees=45
  swing2Motion=Free
  twistLimitDegrees=45
  twistMotion=Free
 }
 {
  bone1Name=dyn_jacket_fr_r_MSTR
  bone2Name=spine_03
  swing1LimitDegrees=45
  swing1Motion=Free
  swing2LimitDegrees=45
  swing2Motion=Free
  twistLimitDegrees=45
  twistMotion=Free
 }
 {
  bone1Name=dyn_jacket_fr_l_MSTR
  bone2Name=spine_03
  swing1LimitDegrees=45
  swing1Motion=Free
  swing2LimitDegrees=45
  swing2Motion=Free
  twistLimitDegrees=45
  twistMotion=Free
 }
 {
  bone1Name=spine_04
  bone2Name=spine_03
  swing1LimitDegrees=45
  swing1Motion=Free
  swing2LimitDegrees=45
  swing2Motion=Free
  twistLimitDegrees=45
  twistMotion=Free
 }
 {bone1Name=dyn_sleeve_r, bone2Name=upperarm_twist_01_r, swing1LimitDegrees=10, swing2LimitDegrees=10, twistLimitDegrees=2}
 {
  bone1Name=upperarm_twist_01_r
  bone2Name=spine_05
  swing1LimitDegrees=45
  swing1Motion=Free
  swing2LimitDegrees=45
  swing2Motion=Free
  twistLimitDegrees=45
  twistMotion=Free
 }
 {
  bone1Name=spine_05
  bone2Name=spine_04
  swing1LimitDegrees=45
  swing1Motion=Free
  swing2LimitDegrees=45
  swing2Motion=Free
  twistLimitDegrees=45
  twistMotion=Free
 }
 {bone1Name=lowerarm_l, bone2Name=spine_05, swing1LimitDegrees=3, swing2LimitDegrees=45, twistLimitDegrees=10}
 {bone1Name=lowerarm_twist_01_l, bone2Name=lowerarm_l, swing1LimitDegrees=3, swing2LimitDegrees=45, twistLimitDegrees=10}
 {bone1Name=Dyn_ArmBand_L3, bone2Name=lowerarm_twist_01_l, swing1LimitDegrees=3, swing2LimitDegrees=45, twistLimitDegrees=10}
 {bone1Name=lowerarm_twist_02_l, bone2Name=lowerarm_l, swing1LimitDegrees=3, swing2LimitDegrees=45, twistLimitDegrees=10}
 {bone1Name=Dyn_ArmBand_L2, bone2Name=lowerarm_twist_02_l, swing1LimitDegrees=3, swing2LimitDegrees=45, twistLimitDegrees=10}
 {bone1Name=Dyn_ArmBand_L1, bone2Name=lowerarm_l, swing1LimitDegrees=3, swing2LimitDegrees=45, twistLimitDegrees=10}
 {
  bone1Name=lowerarm_r
  bone2Name=spine_05
  swing1LimitDegrees=45
  swing1Motion=Free
  swing2LimitDegrees=45
  swing2Motion=Free
  twistLimitDegrees=45
  twistMotion=Free
 }
 {
  bone1Name=lowerarm_twist_01_r
  bone2Name=lowerarm_r
  swing1LimitDegrees=45
  swing1Motion=Free
  swing2LimitDegrees=45
  swing2Motion=Free
  twistLimitDegrees=45
  twistMotion=Free
 }
 {bone1Name=Dyn_ArmBand_R3, bone2Name=lowerarm_twist_01_r, swing1LimitDegrees=3, swing2LimitDegrees=45, twistLimitDegrees=10}
 {
  bone1Name=lowerarm_twist_02_r
  bone2Name=lowerarm_r
  swing1LimitDegrees=45
  swing1Motion=Free
  swing2LimitDegrees=45
  swing2Motion=Free
  twistLimitDegrees=45
  twistMotion=Free
 }
 {bone1Name=Dyn_ArmBand_R2, bone2Name=lowerarm_twist_02_r, swing1LimitDegrees=3, swing2LimitDegrees=45, twistLimitDegrees=10}
 {bone1Name=Dyn_ArmBand_R1, bone2Name=lowerarm_r, swing1LimitDegrees=3, swing2LimitDegrees=45, twistLimitDegrees=10}
]
