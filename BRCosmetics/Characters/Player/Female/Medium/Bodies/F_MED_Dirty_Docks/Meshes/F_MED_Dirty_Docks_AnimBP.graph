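# Post-copy Anim Blueprint for the Dirty Docks female medium cosmetic body, adding secondary physics to the pose copied from its parent mesh.
# Two branches each run hood RigidBody, sleeve and crouch-driven left glute bone offsets, then body RigidBody.
# bIsTargeting selects the branch that replaces sleeve rotations and uses the NoSleeves physics asset; the other adds rotation.
# Hood AnimDynamics chains are left unconnected, and the EventGraph is empty.
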AnimBlueprint /BRCosmetics/Characters/Player/Female/Medium/Bodies/F_MED_Dirty_Docks/Meshes/F_MED_Dirty_Docks_AnimBP
target_skeleton=&/F_MED_Dirty_Docks_Skeleton
graphs [
 AnimGraph {nodes [
   n1 = Anim Dynamics - Bone: dyn_hood_offset(Alpha=0.1, ExternalForce="0,0,0") {
    angularDampingOverride=0.7
    angularSpringConstant=1000
    bGravityOverrideInSimSpace=true
    bOverrideAngularDamping=true
    bOverrideLinearDamping=true
    bUseGravityOverride=true
    linearDampingOverride=0.8
    linearSpringConstant=1000
    simulationSpace=RootRelative
    alphaBoolBlend {blendOption=Linear}
    boundBone {boneName=dyn_hood_offset}
    componentAppliedLinearAccClamp=<30000 30000 30000>
    componentLinearAccScale=<0.75 0.75 0.75>
    componentLinearVelScale=<0.2 0.2 0.2>
    gravityOverride {z=-1100}
    physicsBodyDefinitions [{
      sphereCollisionRadius=0
      boundBone {boneName=dyn_hood_offset}
      boxExtents=<20 20 20>
      constraintSetup {angularLimitsMax=<5 25 10>, angularLimitsMin=<-5 -120 -10>, angularTarget=<0 0 0>, linearAxesMax=<2 2 1>, linearAxesMin=<-2 -2 -1>}
      localJointOffset=<0 0 0>
     }]
    relativeSpaceBone {boneName=root}
   }
   n2 = Anim Dynamics - Bone: dyn_hood_1_1(ComponentPose=@n1, Alpha=0.2, ExternalForce="0,0,0") {
    angularDampingOverride=0.7
    angularSpringConstant=250
    bAngularSpring=true
    bGravityOverrideInSimSpace=true
    bLinearSpring=true
    bOverrideAngularDamping=true
    bOverrideLinearDamping=true
    bUseGravityOverride=true
    linearDampingOverride=0.7
    linearSpringConstant=250
    simulationSpace=RootRelative
    alphaBoolBlend {blendOption=Linear}
    boundBone {boneName=dyn_hood_1_1}
    componentAppliedLinearAccClamp=<30000 30000 30000>
    componentLinearAccScale=<0.1 0.1 0.6>
    componentLinearVelScale=<0.5 0.5 0.4>
    gravityOverride {z=-1100}
    physicsBodyDefinitions [{
      sphereCollisionRadius=0
      boundBone {boneName=dyn_hood_1_1}
      boxExtents=<20 20 20>
      constraintSetup {angularLimitsMax=<5 25 100>, angularLimitsMin=<-5 -35 -10>, angularTarget=<0 0 0>, linearAxesMax=<10 2 0>, linearAxesMin=<-2 -2 0>}
      localJointOffset=<0 0 0>
     }]
    relativeSpaceBone {boneName=root}
   }
   Anim Dynamics - Bone: dyn_hood_1_2(ComponentPose=@n2, Alpha=0.1, ExternalForce="0,0,0") {
    angularDampingOverride=0.7
    angularSpringConstant=250
    bGravityOverrideInSimSpace=true
    bOverrideAngularDamping=true
    bOverrideLinearDamping=true
    bUseGravityOverride=true
    linearDampingOverride=0.8
    linearSpringConstant=250
    alphaBoolBlend {blendOption=Linear}
    boundBone {boneName=dyn_hood_1_2}
    componentAppliedLinearAccClamp=<30000 30000 30000>
    componentLinearAccScale=<0.1 0.1 0.55>
    componentLinearVelScale=<0.7 0.7 0.7>
    gravityOverride {z=-1100}
    physicsBodyDefinitions [{
      sphereCollisionRadius=0
      boundBone {boneName=dyn_hood_1_2}
      boxExtents=<20 20 20>
      constraintSetup {angularLimitsMax=<5 25 10>, angularLimitsMin=<-5 -120 -10>, angularTarget=<0 0 0>, linearAxesMax=<2 1 10>, linearAxesMin=<-2 -1 -5>}
      localJointOffset=<0 0 0>
     }]
    relativeSpaceBone {boneName=spine_05}
   }
   n3 = Anim Dynamics - Bone: dyn_hood_offset(Alpha=0.1, ExternalForce="0,0,0") {
    angularBiasOverride=0.5
    angularDampingOverride=0.8
    angularSpringConstant=250
    bGravityOverrideInSimSpace=true
    bOverrideAngularDamping=true
    bOverrideLinearDamping=true
    bUseGravityOverride=true
    gravityScale=0.2
    linearDampingOverride=0.8
    linearSpringConstant=200
    simulationSpace=RootRelative
    alphaBoolBlend {blendOption=Linear}
    boundBone {boneName=dyn_hood_offset}
    componentAppliedLinearAccClamp=<30000 30000 30000>
    componentLinearAccScale=<0.3 0.3 0.3>
    componentLinearVelScale=<0.6 0.6 0.6>
    gravityOverride {z=-1100}
    physicsBodyDefinitions [{
      sphereCollisionRadius=0
      boundBone {boneName=dyn_hood_offset}
      boxExtents=<20 20 20>
      constraintSetup {angularLimitsMax=<5 25 10>, angularLimitsMin=<-5 -120 -10>, angularTarget=<0 0 0>, linearAxesMax=<2 2 10>, linearAxesMin=<-2 -2 -5>}
      localJointOffset=<0 0 -10>
     }]
    relativeSpaceBone {boneName=root}
   }
   n4 = Anim Dynamics - Bone: dyn_hood_1_1(ComponentPose=@n3, Alpha=0.2, ExternalForce="0,0,0") {
    angularDampingOverride=0.7
    angularSpringConstant=50
    bGravityOverrideInSimSpace=true
    bOverrideAngularDamping=true
    bOverrideLinearDamping=true
    bUseGravityOverride=true
    linearDampingOverride=1
    linearSpringConstant=50
    alphaBoolBlend {blendOption=Linear}
    boundBone {boneName=dyn_hood_1_1}
    componentAppliedLinearAccClamp=<30000 30000 30000>
    componentLinearAccScale=<0.1 0.1 0.6>
    componentLinearVelScale=<0.5 0.5 0.4>
    gravityOverride {z=-1100}
    physicsBodyDefinitions [{
      sphereCollisionRadius=0
      boundBone {boneName=dyn_hood_1_1}
      boxExtents=<20 20 20>
      constraintSetup {angularLimitsMax=<5 25 100>, angularLimitsMin=<-5 -35 -10>, angularTarget=<0 0 0>, linearAxesMax=<10 2 0>, linearAxesMin=<-2 -2 0>}
      localJointOffset=<0 0 0>
     }]
    relativeSpaceBone {boneName=spine_05}
   }
   Anim Dynamics - Bone: dyn_hood_1_2(ComponentPose=@n4, Alpha=0.1, ExternalForce="0,0,0") {
    angularDampingOverride=0.7
    angularSpringConstant=250
    bGravityOverrideInSimSpace=true
    bOverrideAngularDamping=true
    bOverrideLinearDamping=true
    bUseGravityOverride=true
    gravityScale=0.3
    linearDampingOverride=0.9
    linearSpringConstant=250
    simulationSpace=RootRelative
    alphaBoolBlend {blendOption=Linear}
    boundBone {boneName=dyn_hood_1_2}
    componentAppliedLinearAccClamp=<30000 30000 30000>
    componentLinearAccScale=<0.1 0.1 0.55>
    componentLinearVelScale=<0.7 0.7 0.7>
    gravityOverride {z=-1100}
    physicsBodyDefinitions [{
      sphereCollisionRadius=0
      boundBone {boneName=dyn_hood_1_2}
      boxExtents=<20 20 20>
      constraintSetup {angularLimitsMax=<5 25 10>, angularLimitsMin=<-5 -120 -10>, angularTarget=<0 0 0>, linearAxesMax=<2 2 10>, linearAxesMin=<-2 -2 -5>}
      localJointOffset=<0 0 0>
     }]
    relativeSpaceBone {boneName=root}
   }
   n5 = RigidBody(ComponentPose=Local To Component(LocalPose=Copy Pose From Mesh() {bUseAttachedParent=true}), Alpha=0.45) {
    bOverrideWorldGravity=true
    overridePhysicsAsset=&/F_MED_Dirty_Docks_Physics_Hood
    simulationSpace=BaseBoneSpace
    alphaBoolBlend {blendOption=Linear}
    baseBoneRef {boneName=neck_02}
    componentLinearAccScale=<0.2 0.2 0.2>
    componentLinearVelScale=<0.6 0.6 0.6>
    overrideWorldGravity {z=-980}
   }
   n6 = "Transform (Modify) Bone - Bone: dyn_sleeve_r"(ComponentPose=@n5, Alpha=1, Translation="0,0,0", Rotation="14,-2,-15", Scale="1,1,1") {
    rotationMode=BMM_Additive
    rotationSpace=BCS_ParentBoneSpace
    alphaBoolBlend {blendOption=Linear}
    boneToModify {boneName=dyn_sleeve_r}
    rotation=<rot 14 -2 -15>
   }
   n7 = "Transform (Modify) Bone - Bone: dyn_sleeve_l"(ComponentPose=@n6, Alpha=1, Translation="0,0,0", Rotation="20,-14,-8", Scale="1,1,1") {
    rotationMode=BMM_Additive
    rotationSpace=BCS_ParentBoneSpace
    alphaBoolBlend {blendOption=Linear}
    boneToModify {boneName=dyn_sleeve_l}
    rotation=<rot 20 -14 -8>
   }
   n8 = "Transform (Modify) Bone - Bone: dyn_glute_l"(ComponentPose=@n7, Alpha="To Float (Boolean)"(InBool=Get bIsCrouching()), Translation="0.000,-4,-3", Rotation="0,0,0", Scale="1,1,1") {translationMode=BMM_Additive, alphaBoolBlend {blendOption=Linear}, boneToModify {boneName=dyn_glute_l}, translation {y=-4, z=-3}}
   n9 = RigidBody(ComponentPose=@n8) {
    bOverrideWorldGravity=true
    overridePhysicsAsset=&/F_MED_Dirty_Docks_Physics
    simulationSpace=BaseBoneSpace
    alphaBoolBlend {blendOption=Linear}
    baseBoneRef {boneName=root}
    componentLinearAccScale=<0.4 0.4 0.4>
    componentLinearVelScale=<0.7 0.7 0.7>
    overrideWorldGravity {z=-980}
   }
   n10 = RigidBody(ComponentPose=Local To Component(LocalPose=Copy Pose From Mesh() {bUseAttachedParent=true}), Alpha=0.6) {
    bOverrideWorldGravity=true
    overridePhysicsAsset=&F_MED_Dirty_Docks_Physics_Hood
    simulationSpace=BaseBoneSpace
    alphaBoolBlend {blendOption=Linear}
    baseBoneRef {boneName=spine_05}
    componentLinearAccScale=<0.3 0.3 0.3>
    componentLinearVelScale=<0.6 0.6 0.6>
    overrideWorldGravity {z=-980}
   }
   n11 = "Transform (Modify) Bone - Bone: dyn_glute_l"(ComponentPose=@n10, Alpha="To Float (Boolean)"(InBool=Get bIsCrouching()), Translation="0,-4,-3", Rotation="0,0,0", Scale="1,1,1") {translationMode=BMM_Additive, alphaBoolBlend {blendOption=Linear}, boneToModify {boneName=dyn_glute_l}, translation {y=-4, z=-3}}
   n12 = "Transform (Modify) Bone - Bone: dyn_sleeve_r"(ComponentPose=@n11, Alpha=1, Translation="0,0,0", Rotation="35,-20,30", Scale="1,1,1") {
    rotationMode=BMM_Replace
    rotationSpace=BCS_ParentBoneSpace
    alphaBoolBlend {blendOption=Linear}
    boneToModify {boneName=dyn_sleeve_r}
    rotation=<rot 35 -20 30>
   }
   n13 = "Transform (Modify) Bone - Bone: dyn_sleeve_l"(ComponentPose=@n12, Alpha=1, Translation="0,0,0", Rotation="40,-28.000,-20", Scale="1,1,1") {
    rotationMode=BMM_Replace
    rotationSpace=BCS_ParentBoneSpace
    alphaBoolBlend {blendOption=Linear}
    boneToModify {boneName=dyn_sleeve_l}
    rotation=<rot 40 -28 -20>
   }
   n14 = RigidBody(ComponentPose=@n13, Alpha=0.5) {
    bOverrideWorldGravity=true
    overridePhysicsAsset=&/F_MED_Dirty_Docks_Physics_NoSleeves
    simulationSpace=BaseBoneSpace
    alphaBoolBlend {blendOption=Linear}
    baseBoneRef {boneName=root}
    componentLinearAccScale=<0.35 0.35 0.35>
    componentLinearVelScale=<0.35 0.35 0.35>
    overrideWorldGravity {z=-980}
   }
   Output Pose(Result=Blend Poses by bool(BlendPose_0=Component To Local(ComponentPose=@n14), BlendPose_1=Component To Local(ComponentPose=@n9), bActiveValue=Get bIsTargeting()) {blendType=Linear}) {layerGroup=None}
  ]}
 EventGraph {nodes [Event BlueprintUpdateAnimation()]}
]
variables []
