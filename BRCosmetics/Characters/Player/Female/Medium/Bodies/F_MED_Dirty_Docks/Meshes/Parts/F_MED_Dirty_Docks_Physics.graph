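# Physics asset for the female Dirty Docks cosmetic body, giving secondary motion to hair, mask hair, nose ring, fin and tubes.
# Capsules cover head, neck, upper spine and right shoulder, while boxes and small capsules form the dynamic chains.
# Chains hang from the head, except the lower tube on spine_05 and hair branches off other strands.
# Constraints mostly use tight angular limits, with free body joints, tube masters and several mask-hair roots.
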PhysicsAsset /BRCosmetics/Characters/Player/Female/Medium/Bodies/F_MED_Dirty_Docks/Meshes/Parts/F_MED_Dirty_Docks_Physics
bodies {
 DYN_Nose_Ring [{shapeType=Box, extentX=1.5805, extentY=2.0796, extentZ=0.5, center=<0.8098 0 0>}]
 Dyn_Hair_01_1 [{shapeType=Box, extentX=2.3324, extentY=1, extentZ=2.4982, center=<1.2786 -0.0001 0>}]
 Dyn_Hair_01_2 [{shapeType=Box, extentX=2.5201, extentY=0.6824, extentZ=2.535, center=<1.1894 0.0001 0>, rotation=<rot 0 0 -39.9434>}]
 Dyn_Hair_01_3 [{shapeType=Box, extentX=2.2678, extentY=0.5462, extentZ=2.4159, center=<1.2079 0 0>, rotation=<rot 0 0 -28.2419>}]
 Dyn_Hair_01_4 [{shapeType=Box, extentX=1.735, extentY=0.7809, extentZ=2.0936, center=<0.9078 0 0>, rotation=<rot 0.0001 0 -14.693>}]
 Dyn_Hair_02_1 [{shapeType=Box, extentX=2.1889, extentY=1, extentZ=2.5476, center=<1.1998 0 0>, rotation=<rot 0 0 -27.4697>}]
 Dyn_Hair_02_2 [{shapeType=Box, extentX=3.4295, extentY=0.6205, extentZ=2.9715, center=<1.7653 0 0>, rotation=<rot -0.0001 -0.0001 -23.9785>}]
 Dyn_Hair_02_3 [{shapeType=Box, extentX=2.9785, extentY=1, extentZ=3.2558, center=<1.5675 0 0.0001>, rotation=<rot 0 0 -22.3836>}]
 Dyn_Hair_02_4 [{shapeType=Box, extentX=2.7787, extentY=1, extentZ=3.3306, center=<1.4266 0 0.0001>, rotation=<rot 0.0001 0 -10.5143>}]
 Dyn_Hair_03_1 [{shapeType=Box, extentX=2.5919, extentY=2.8973, extentZ=5.855, center=<1.2769 0.3039 0.058>, rotation=<rot 0 0 -10.7934>}]
 Dyn_Hair_03_2 [{shapeType=Box, extentX=4.6335, extentY=3.3075, extentZ=5.6114, center=<2.4053 0 0.0001>, rotation=<rot 0 0 -14.3913>}]
 Dyn_Hair_03_3 [{shapeType=Box, extentX=4.25, extentY=1.5809, extentZ=5.5177, center=<2.2419 0.0465 0.0108>, rotation=<rot 0 0 -14.076>}]
 Dyn_Hair_03_4 [{shapeType=Box, extentX=3.1837, extentY=1, extentZ=3.9034, center=<1.5565 0 0>, rotation=<rot 0 0 -13.8448>}]
 Dyn_Hair_03_5 [{shapeType=Box, extentX=2.5672, extentY=1, extentZ=3.2377, center=<1.2735 0 0>, rotation=<rot 0 0 -15.8624>}]
 Dyn_Hair_03_6 [{shapeType=Box, extentX=1.7789, extentY=1, extentZ=3.0035, center=<0.9384 0.0001 0>}]
 Dyn_Hair_04_1 [{shapeType=Box, extentX=2.294, extentY=2.9006, extentZ=6.4626, center=<1.1144 0.7419 -0.219>, rotation=<rot 0 0 16.4482>}]
 Dyn_Hair_04_2 [{shapeType=Box, extentX=6.6791, extentY=2.6559, extentZ=6.6208, center=<3.4956 0.5697 -0.1589>, rotation=<rot 0 0 15.5751>}]
 Dyn_Hair_04_3 [{shapeType=Box, extentX=5.0257, extentY=3.267, extentZ=5.5872, center=<2.5398 0.0762 -0.011>, rotation=<rot 0 0 8.1558>}]
 Dyn_Hair_04_4 [{shapeType=Box, extentX=3.7208, extentY=2.6907, extentZ=5.4753, center=<1.8817 0 -0.0001>, rotation=<rot 0 0 21.9362>}]
 Dyn_Hair_04_5 [{shapeType=Box, extentX=3.8225, extentY=2.1466, extentZ=4.6281, center=<1.9338 -0.0001 0>, rotation=<rot 0 0 16.2188>}]
 Dyn_Hair_04_6 [{shapeType=Box, extentX=2.6965, extentY=1, extentZ=4.5697, center=<1.4076 0 0>, rotation=<rot 0 0 31.2792>}]
 Dyn_Hair_05_1 [{
   shapeType=Capsule
   length=0.6
   radius=0.5
   shapeName=Dyn_MaskHair_05_1_capsule
   center=<1.1639 -0.2856 0.0146>
   rotation=<rot 0 90.108 -89.4948>
  }]
 Dyn_Hair_05_2 [{
   shapeType=Capsule
   length=0.505
   radius=0.505
   shapeName=Dyn_MaskHair_05_2_capsule
   center=<1.2936 -0.5266 0.0405>
   rotation=<rot 0.5116 -100.6192 -88.5698>
  }]
 Dyn_Hair_06_1 [{shapeType=Capsule, length=0.0586, radius=0.7731, center=<0.7351 0 0>, rotation=<rot 90 0 0>}]
 Dyn_Hair_06_2 [{shapeType=Capsule, length=0.1737, radius=0.5941, center=<0.6397 -0.0001 0>, rotation=<rot 90 0 0>}]
 Dyn_Hair_07_1 [{shapeType=Capsule, length=0.1429, radius=0.6292, center=<0.5305 0 -0.0001>, rotation=<rot 90 0 0>}]
 Dyn_Hair_07_2 [{shapeType=Capsule, length=0, radius=0.6293, center=<0.6964 0 0>, rotation=<rot 90 0 0>}]
 Dyn_MaskHair_01_1 [{shapeType=Box, extentX=2.5625, extentY=1, extentZ=2.4375, center=<1.3438 0 0>}]
 Dyn_MaskHair_01_2 [{shapeType=Box, extentX=2.625, extentY=1, extentZ=2.3125, center=<1.4199 0 0>, rotation=<rot 0 0 5.6253>}]
 Dyn_MaskHair_01_3 [{shapeType=Box, extentX=2.3125, extentY=1, extentZ=1.9375, center=<1.251 0 0>}]
 Dyn_MaskHair_01_4 [{shapeType=Box, extentX=1.3125, extentY=1, extentZ=1.6875, center=<0.7002 0 0>}]
 Dyn_MaskHair_02_1 [{shapeType=Box, extentX=2.4375, extentY=1, extentZ=2.5, center=<1.2263 0 0>}]
 Dyn_MaskHair_02_2 [{shapeType=Box, extentX=3.5, extentY=1, extentZ=2.625, center=<1.7423 0 0>}]
 Dyn_MaskHair_02_3 [{shapeType=Box, extentX=2.6875, extentY=1, extentZ=2.5625, center=<1.4428 0 0>}]
 Dyn_MaskHair_02_4 [{shapeType=Box, extentX=2.25, extentY=1.4375, extentZ=2.25, center=<1.1434 0 0>}]
 Dyn_MaskHair_02_5 [{shapeType=Box, extentX=1.8125, extentY=1, extentZ=1.9375, center=<0.8616 0 0>}]
 Dyn_MaskHair_03_1 [{shapeType=Box, extentX=2.75, extentY=1, extentZ=7.3125, center=<1.2318 0.0004 0>}]
 Dyn_MaskHair_03_2 [{shapeType=Box, extentX=4.8125, extentY=1, extentZ=6.25, center=<2.5481 0 0>}]
 Dyn_MaskHair_03_3 [{shapeType=Box, extentX=4.4375, extentY=1, extentZ=6.125, center=<2.3819 0.0003 0>, rotation=<rot 0 0 -5.6264>}]
 Dyn_MaskHair_03_4 [{shapeType=Box, extentX=3.25, extentY=1.8125, extentZ=6, center=<1.6171 0 0>, rotation=<rot 0 0 -2.8126>}]
 Dyn_MaskHair_03_5 [{shapeType=Box, extentX=3, extentY=1, extentZ=3.5625, center=<1.5397 0 0>, rotation=<rot 0 0 -5.6252>}]
 Dyn_MaskHair_03_6 [{shapeType=Box, extentX=2.625, extentY=1, extentZ=2.125, center=<1.3117 0.3096 0>, rotation=<rot 0 14.0642 0>}]
 Dyn_MaskHair_04_1 [{shapeType=Box, extentX=2.4375, extentY=1.625, extentZ=7.4375, center=<1.4549 0 0>, rotation=<rot 0 0 -11.2534>}]
 Dyn_MaskHair_04_2 [{shapeType=Box, extentX=6.8125, extentY=1, extentZ=7.6875, center=<3.4495 0 0>, rotation=<rot 0 0 11.2512>}]
 Dyn_MaskHair_04_3 [{shapeType=Box, extentX=5, extentY=0.5, extentZ=6.6875, center=<2.6505 -0.4955 0.0487>, rotation=<rot 0 0 5.6265>}]
 Dyn_MaskHair_04_4 [{shapeType=Box, extentX=3.6875, extentY=0.5, extentZ=6.125, center=<1.8889 -0.2194 0.0325>, rotation=<rot 0 0 8.4386>}]
 Dyn_MaskHair_04_5 [{shapeType=Box, extentX=3.875, extentY=0.875, extentZ=3.1875, center=<1.9119 0 0>, rotation=<rot 0 0 -14.064>}]
 Dyn_MaskHair_04_6 [{shapeType=Box, extentX=2.5, extentY=1, extentZ=1.6875, center=<1.2273 0 0>}]
 Dyn_MaskHair_05_1 [{
   shapeType=Capsule
   length=0.6074
   radius=0.4093
   shapeName=Dyn_MaskHair_05_1_capsule
   center=<0.837 -1.5128 0.0116>
   rotation=<rot 0.1466 73.2337 -89.5166>
  }]
 Dyn_MaskHair_05_2 [{
   shapeType=Capsule
   length=0.505
   radius=0.505
   shapeName=Dyn_MaskHair_05_2_capsule
   center=<0.9297 -2.4047 0.0579>
   rotation=<rot 0.6493 -106.2428 -88.627>
  }]
 Dyn_MaskHair_07_1 [{shapeType=Capsule, length=0.1429, radius=0.6292, center=<0.5305 0 -0.0001>, rotation=<rot 90 0 0>}]
 Dyn_MaskHair_07_2 [{shapeType=Capsule, length=0, radius=0.6293, center=<0.6964 0 0>, rotation=<rot 90 0 0>}]
 clavicle_r [{shapeType=Capsule, length=6.125, radius=5.7863, center=<-13.4816 0 2.5279>, rotation=<rot 0 89.9999 81.562>}]
 dyn_fin_01 [{shapeType=Box, extentX=2.875, extentY=7.5, extentZ=1, center=<1.6367 0 0>}]
 dyn_fin_02 [{shapeType=Box, extentX=5.875, extentY=5.5, extentZ=1, center=<3.0575 0 0>}]
 dyn_fin_03 [{shapeType=Box, extentX=2.375, extentY=4.8125, extentZ=1, center=<1.2303 0 0>}]
 dyn_lw_tube_02 [{shapeType=Box, extentX=4.5434, extentY=3.1262, extentZ=4.8742, center=<2.4072 0.0003 -0.6917>, rotation=<rot -16.8753 0 0>}]
 dyn_lw_tube_03 [{shapeType=Box, extentX=3.5882, extentY=9.2934, extentZ=8.464, center=<4.3544 -0.3232 -0.5036>, rotation=<rot 24.9122 -89.7823 -79.0996>}]
 dyn_lw_tube_MSTR [{shapeType=Capsule, length=0.505, radius=0.505, rotation=<rot 0 -90 0>}]
 dyn_up_tube_03 [{shapeType=Box, extentX=8.6383, extentY=2.8622, extentZ=5.4952, center=<2.8327 0.5988 0.7494>, rotation=<rot 0 89.7502 -89.2577>}]
 dyn_up_tube_04 [{shapeType=Box, extentX=4.4181, extentY=2.5, extentZ=11.1313, center=<5.6 2.3586 1.4744>, rotation=<rot 2.8124 -78.2164 -90.6714>}]
 dyn_up_tube_MSTR [{shapeType=Capsule, length=0.505, radius=0.505, rotation=<rot 0 -90 0>}]
 head [
  {shapeType=Capsule, length=3.1125, radius=8.7677, center=<3.3016 1.3529 -1.3293>, rotation=<rot 90 -90 -77.7357>}
  {shapeType=Capsule, length=3.5329, radius=1.9946, center=<2.5264 7.3707 2.7325>, rotation=<rot 0 0 -12.5974>}
  {shapeType=Capsule, length=1.2988, radius=1.9735, center=<2.5856 -0.4973 5.4482>, rotation=<rot 19.9563 74.9961 110.1947>}
 ]
 neck_01 [{shapeType=Capsule, length=9.75, radius=12.4113, center=<2.5878 0.0002 -2.8958>, rotation=<rot 90 0 0>}]
 neck_02 [{shapeType=Capsule, length=0.0625, radius=4.1613, center=<2.5878 0.0002 0>, rotation=<rot 90 0 0>}]
 spine_05 [
  {shapeType=Capsule, length=39.6893, radius=16.1207, center=<-9.1854 9.1902 0>, rotation=<rot 0 90 0>}
  {shapeType=Capsule, length=39.6893, radius=16.1207, center=<2.3798 27.8931 -0.0001>, rotation=<rot 0 90 0>}
 ]
 upperarm_r [
  {shapeType=Capsule, length=5.75, radius=6.505, center=<-4.6895 0 1.1434>, rotation=<rot 90 0 0>}
  {shapeType=Capsule, length=12.875, radius=9.25, center=<-2.085 -7.7417 3.0984>, rotation=<rot -70.2509 -21.8679 141.9077>}
 ]
}
constraints [
 {bone1Name=Dyn_Hair_05_2, bone2Name=Dyn_Hair_05_1, swing1LimitDegrees=25, swing2LimitDegrees=10, twistLimitDegrees=10}
 {bone1Name=DYN_Nose_Ring, bone2Name=head, swing1LimitDegrees=20, swing2LimitDegrees=30, twistLimitDegrees=5}
 {bone1Name=Dyn_Hair_01_1, bone2Name=head, swing1LimitDegrees=25, swing2LimitDegrees=10, twistLimitDegrees=10}
 {bone1Name=Dyn_Hair_01_2, bone2Name=Dyn_Hair_01_1, swing1LimitDegrees=10, swing2LimitDegrees=10, twistLimitDegrees=10}
 {bone1Name=Dyn_Hair_01_3, bone2Name=Dyn_Hair_01_2, swing1LimitDegrees=25, swing2LimitDegrees=10, twistLimitDegrees=10}
 {bone1Name=Dyn_Hair_01_4, bone2Name=Dyn_Hair_01_3, swing1LimitDegrees=25, swing2LimitDegrees=10, twistLimitDegrees=10}
 {
  bone1Name=Dyn_Hair_03_1
  bone2Name=head
  swing1LimitDegrees=25
  swing1Motion=Locked
  swing2LimitDegrees=6
  swing2Motion=Locked
  twistLimitDegrees=10
  twistMotion=Locked
 }
 {bone1Name=Dyn_Hair_03_2, bone2Name=Dyn_Hair_03_1, swing1LimitDegrees=10, swing1Motion=Locked, swing2LimitDegrees=5, twistLimitDegrees=3}
 {bone1Name=Dyn_Hair_03_3, bone2Name=Dyn_Hair_03_2, swing1LimitDegrees=1, swing2LimitDegrees=5, twistLimitDegrees=3}
 {bone1Name=Dyn_Hair_03_4, bone2Name=Dyn_Hair_03_3, swing1LimitDegrees=25, swing2LimitDegrees=10, twistLimitDegrees=5}
 {bone1Name=Dyn_Hair_03_5, bone2Name=Dyn_Hair_03_4, swing1LimitDegrees=25, swing2LimitDegrees=10, twistLimitDegrees=5}
 {bone1Name=Dyn_Hair_03_6, bone2Name=Dyn_Hair_03_5, swing1LimitDegrees=25, swing2LimitDegrees=10, twistLimitDegrees=10}
 {bone1Name=Dyn_Hair_02_1, bone2Name=Dyn_Hair_03_2, swing1LimitDegrees=25, swing2LimitDegrees=12, twistLimitDegrees=10}
 {bone1Name=Dyn_Hair_02_2, bone2Name=Dyn_Hair_02_1, swing1LimitDegrees=10, swing2LimitDegrees=6, twistLimitDegrees=10}
 {bone1Name=Dyn_Hair_02_3, bone2Name=Dyn_Hair_02_2, swing1LimitDegrees=25, swing2LimitDegrees=6, twistLimitDegrees=10}
 {bone1Name=Dyn_Hair_02_4, bone2Name=Dyn_Hair_02_3, swing1LimitDegrees=25, swing2LimitDegrees=6, twistLimitDegrees=10}
 {bone1Name=Dyn_Hair_04_1, bone2Name=head, swing1LimitDegrees=25, swing2LimitDegrees=6, twistLimitDegrees=10}
 {bone1Name=Dyn_Hair_04_2, bone2Name=Dyn_Hair_04_1, swing1LimitDegrees=10, swing2LimitDegrees=6, twistLimitDegrees=10}
 {bone1Name=Dyn_Hair_04_3, bone2Name=Dyn_Hair_04_2, swing1LimitDegrees=10, swing2LimitDegrees=6, twistLimitDegrees=10}
 {bone1Name=Dyn_Hair_04_4, bone2Name=Dyn_Hair_04_3, swing1LimitDegrees=25, swing2LimitDegrees=6, twistLimitDegrees=10}
 {bone1Name=Dyn_Hair_04_5, bone2Name=Dyn_Hair_04_4, swing1LimitDegrees=25, swing2LimitDegrees=10, twistLimitDegrees=10}
 {bone1Name=Dyn_Hair_04_6, bone2Name=Dyn_Hair_04_5, swing1LimitDegrees=25, swing2LimitDegrees=10, twistLimitDegrees=10}
 {bone1Name=Dyn_Hair_05_1, bone2Name=Dyn_Hair_04_3, swing1LimitDegrees=25, swing2LimitDegrees=10, twistLimitDegrees=10}
 {bone1Name=Dyn_Hair_06_1, bone2Name=head, swing1LimitDegrees=5, swing2LimitDegrees=5, twistLimitDegrees=5}
 {bone1Name=Dyn_Hair_06_2, bone2Name=Dyn_Hair_06_1, swing1LimitDegrees=5, swing2LimitDegrees=5, twistLimitDegrees=5}
 {bone1Name=Dyn_Hair_07_1, bone2Name=head, swing1LimitDegrees=5, swing2LimitDegrees=5, twistLimitDegrees=5}
 {bone1Name=Dyn_Hair_07_2, bone2Name=Dyn_Hair_07_1, swing1LimitDegrees=5, swing2LimitDegrees=5, twistLimitDegrees=5}
 {bone1Name=dyn_fin_01, bone2Name=head, swing1LimitDegrees=0, swing2LimitDegrees=3, twistLimitDegrees=3}
 {bone1Name=dyn_fin_02, bone2Name=dyn_fin_01, swing1LimitDegrees=0, swing2LimitDegrees=3, twistLimitDegrees=3}
 {bone1Name=dyn_fin_03, bone2Name=dyn_fin_02, swing1LimitDegrees=0, swing2LimitDegrees=5, twistLimitDegrees=5}
 {bone1Name=Dyn_MaskHair_05_2, bone2Name=Dyn_MaskHair_05_1, swing1LimitDegrees=25, swing2LimitDegrees=10, twistLimitDegrees=10}
 {bone1Name=Dyn_MaskHair_05_1, bone2Name=Dyn_MaskHair_04_3, swing1LimitDegrees=25, swing2LimitDegrees=25, twistLimitDegrees=10}
 {bone1Name=Dyn_MaskHair_07_1, bone2Name=head, swing1LimitDegrees=5, swing2LimitDegrees=5, twistLimitDegrees=5}
 {bone1Name=Dyn_MaskHair_07_2, bone2Name=Dyn_MaskHair_07_1, swing1LimitDegrees=5, swing2LimitDegrees=5, twistLimitDegrees=5}
 {bone1Name=Dyn_MaskHair_04_6, bone2Name=Dyn_MaskHair_04_5, swing1LimitDegrees=25, swing2LimitDegrees=10, twistLimitDegrees=10}
 {
  bone1Name=Dyn_MaskHair_01_1
  bone2Name=head
  swing1LimitDegrees=45
  swing1Motion=Free
  swing2LimitDegrees=45
  swing2Motion=Free
  twistLimitDegrees=45
  twistMotion=Free
 }
 {bone1Name=Dyn_MaskHair_01_2, bone2Name=Dyn_MaskHair_01_1, swing1LimitDegrees=10, swing2LimitDegrees=10, twistLimitDegrees=10}
 {bone1Name=Dyn_MaskHair_01_3, bone2Name=Dyn_MaskHair_01_2, swing1LimitDegrees=25, swing2LimitDegrees=10, twistLimitDegrees=10}
 {bone1Name=Dyn_MaskHair_01_4, bone2Name=Dyn_MaskHair_01_3, swing1LimitDegrees=25, swing2LimitDegrees=10, twistLimitDegrees=10}
 {
  bone1Name=Dyn_MaskHair_03_1
  bone2Name=head
  swing1LimitDegrees=45
  swing1Motion=Free
  swing2LimitDegrees=45
  swing2Motion=Free
  twistLimitDegrees=45
  twistMotion=Free
 }
 {
  bone1Name=Dyn_MaskHair_03_2
  bone2Name=Dyn_MaskHair_03_1
  swing1LimitDegrees=45
  swing1Motion=Locked
  swing2LimitDegrees=5
  twistLimitDegrees=3
 }
 {bone1Name=Dyn_MaskHair_03_3, bone2Name=Dyn_MaskHair_03_2, swing1LimitDegrees=1, swing2LimitDegrees=5, twistLimitDegrees=3}
 {bone1Name=Dyn_MaskHair_03_4, bone2Name=Dyn_MaskHair_03_3, swing1LimitDegrees=25, swing2LimitDegrees=10, twistLimitDegrees=5}
 {bone1Name=Dyn_MaskHair_03_5, bone2Name=Dyn_MaskHair_03_4, swing1LimitDegrees=25, swing2LimitDegrees=10, twistLimitDegrees=5}
 {bone1Name=Dyn_MaskHair_03_6, bone2Name=Dyn_MaskHair_03_5, swing1LimitDegrees=25, swing2LimitDegrees=10, twistLimitDegrees=10}
 {
  bone1Name=Dyn_MaskHair_02_1
  bone2Name=Dyn_MaskHair_03_2
  swing1LimitDegrees=45
  swing1Motion=Free
  swing2LimitDegrees=45
  swing2Motion=Free
  twistLimitDegrees=45
  twistMotion=Free
 }
 {bone1Name=Dyn_MaskHair_02_2, bone2Name=Dyn_MaskHair_02_1, swing1LimitDegrees=10, swing2LimitDegrees=6, twistLimitDegrees=10}
 {bone1Name=Dyn_MaskHair_02_3, bone2Name=Dyn_MaskHair_02_2, swing1LimitDegrees=10, swing2LimitDegrees=6, twistLimitDegrees=10}
 {bone1Name=Dyn_MaskHair_02_4, bone2Name=Dyn_MaskHair_02_3, swing1LimitDegrees=25, swing2LimitDegrees=6, twistLimitDegrees=10}
 {bone1Name=Dyn_MaskHair_02_5, bone2Name=Dyn_MaskHair_02_4, swing1LimitDegrees=25, swing2LimitDegrees=6, twistLimitDegrees=10}
 {
  bone1Name=Dyn_MaskHair_04_1
  bone2Name=head
  swing1LimitDegrees=45
  swing1Motion=Free
  swing2LimitDegrees=45
  swing2Motion=Free
  twistLimitDegrees=45
  twistMotion=Free
 }
 {bone1Name=Dyn_MaskHair_04_2, bone2Name=Dyn_MaskHair_04_1, swing1LimitDegrees=10, swing2LimitDegrees=6, twistLimitDegrees=10}
 {bone1Name=Dyn_MaskHair_04_3, bone2Name=Dyn_MaskHair_04_2, swing1LimitDegrees=10, swing2LimitDegrees=5, twistLimitDegrees=10}
 {bone1Name=Dyn_MaskHair_04_4, bone2Name=Dyn_MaskHair_04_3, swing1LimitDegrees=25, swing2LimitDegrees=6, twistLimitDegrees=10}
 {bone1Name=Dyn_MaskHair_04_5, bone2Name=Dyn_MaskHair_04_4, swing1LimitDegrees=25, swing2LimitDegrees=10, twistLimitDegrees=10}
 {
  bone1Name=head
  bone2Name=neck_02
  swing1LimitDegrees=45
  swing1Motion=Free
  swing2LimitDegrees=45
  swing2Motion=Free
  twistLimitDegrees=45
  twistMotion=Free
 }
 {
  bone1Name=neck_02
  bone2Name=neck_01
  swing1LimitDegrees=45
  swing1Motion=Free
  swing2LimitDegrees=45
  swing2Motion=Free
  twistLimitDegrees=45
  twistMotion=Free
 }
 {
  bone1Name=upperarm_r
  bone2Name=clavicle_r
  swing1LimitDegrees=45
  swing1Motion=Free
  swing2LimitDegrees=45
  swing2Motion=Free
  twistLimitDegrees=45
  twistMotion=Free
 }
 {
  bone1Name=dyn_up_tube_MSTR
  bone2Name=head
  swing1LimitDegrees=45
  swing1Motion=Free
  swing2LimitDegrees=45
  swing2Motion=Free
  twistLimitDegrees=45
  twistMotion=Free
 }
 {bone1Name=dyn_up_tube_03, bone2Name=dyn_up_tube_MSTR, swing1LimitDegrees=8, swing2LimitDegrees=8, twistLimitDegrees=5}
 {bone1Name=dyn_up_tube_04, bone2Name=dyn_up_tube_03, swing1LimitDegrees=7, swing2LimitDegrees=7, twistLimitDegrees=5}
 {bone1Name=dyn_lw_tube_02, bone2Name=dyn_lw_tube_MSTR, swing1LimitDegrees=5, swing2LimitDegrees=5, twistLimitDegrees=3}
 {bone1Name=dyn_lw_tube_03, bone2Name=dyn_lw_tube_02, swing1LimitDegrees=6, swing2LimitDegrees=6, twistLimitDegrees=5}
 {
  bone1Name=clavicle_r
  bone2Name=spine_05
  swing1LimitDegrees=45
  swing1Motion=Free
  swing2LimitDegrees=45
  swing2Motion=Free
  twistLimitDegrees=45
  twistMotion=Free
 }
 {
  bone1Name=neck_01
  bone2Name=spine_05
  swing1LimitDegrees=45
  swing1Motion=Free
  swing2LimitDegrees=45
  swing2Motion=Free
  twistLimitDegrees=45
  twistMotion=Free
 }
 {
  bone1Name=dyn_lw_tube_MSTR
  bone2Name=spine_05
  swing1LimitDegrees=45
  swing1Motion=Free
  swing2LimitDegrees=45
  swing2Motion=Free
  twistLimitDegrees=45
  twistMotion=Free
 }
]
