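# Post-copy AnimBP for the female Dirty Docks body part that keeps its upper and lower tubes following the character's head and aiming.
# AnimGraph copies the parent mesh pose, runs RigidBody physics relative to neck_02, then layers additive bone rotations on the tubes.
# EventGraph derives those alphas from head rotation against spine_05 and bIsTargeting.
# Arm-raise and head_r_more values are computed but unused; one ADS bone node is unconnected.
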AnimBlueprint /BRCosmetics/Characters/Player/Female/Medium/Bodies/F_MED_Dirty_Docks/Meshes/Parts/F_MED_Dirty_Docks_AnimBP
target_skeleton=&/F_MED_Dirty_Docks_Skeleton
graphs [
 AnimGraph {nodes [
   n1 = Get is_ADS() -> (is_ADS:double)
   "Transform (Modify) Bone - Bone: dyn_up_tube_02"(Alpha=@n1, Translation="0,0,0", Rotation="8,-2,0", Scale="1,1,1") {
    rotationMode=BMM_Additive
    rotationSpace=BCS_BoneSpace
    scaleMode=BMM_Additive
    scaleSpace=BCS_BoneSpace
    translationMode=BMM_Additive
    translationSpace=BCS_BoneSpace
    alphaBoolBlend {blendOption=Linear}
    alphaScaleBiasClamp {bInterpResult=true, interpSpeedDecreasing=30, interpSpeedIncreasing=30}
    boneToModify {boneName=dyn_up_tube_02}
    rotation {pitch=8, yaw=-2}
   }
   n2 = RigidBody(ComponentPose=Local To Component(LocalPose=Copy Pose From Mesh() {bUseAttachedParent=true}), Alpha=0.8) {
    bOverrideWorldGravity=true
    overridePhysicsAsset=&/F_MED_Dirty_Docks_Physics
    simulationSpace=BaseBoneSpace
    alphaBoolBlend {blendOption=Linear}
    baseBoneRef {boneName=neck_02}
    componentLinearAccScale=<0.3 0.3 0.3>
    componentLinearVelScale=<0.6 0.6 0.6>
    overrideWorldGravity {z=-980}
   }
   n3 = Get head_l() -> (head_l:double)
   n4 = "Transform (Modify) Bone - Bone: dyn_lw_tube_MSTR"(ComponentPose=@n2, Alpha=@n3, Translation="0,0,0", Rotation="-10,-40,0.000", Scale="1,1,1") {
    rotationMode=BMM_Additive
    rotationSpace=BCS_BoneSpace
    scaleMode=BMM_Additive
    scaleSpace=BCS_BoneSpace
    translationMode=BMM_Additive
    translationSpace=BCS_BoneSpace
    alphaBoolBlend {blendOption=Linear}
    alphaScaleBiasClamp {interpSpeedDecreasing=30, interpSpeedIncreasing=30}
    boneToModify {boneName=dyn_lw_tube_MSTR}
    rotation {pitch=-10, yaw=-40}
   }
   n5 = "Transform (Modify) Bone - Bone: dyn_up_tube_MSTR"(ComponentPose=@n4, Alpha=@n3, Translation="0,0,0", Rotation="0.000,20,0.000", Scale="1,1,1") {
    rotationMode=BMM_Additive
    rotationSpace=BCS_BoneSpace
    scaleMode=BMM_Additive
    scaleSpace=BCS_BoneSpace
    translationMode=BMM_Additive
    translationSpace=BCS_BoneSpace
    alphaBoolBlend {blendOption=Linear}
    alphaScaleBiasClamp {interpSpeedDecreasing=30, interpSpeedIncreasing=30}
    boneToModify {boneName=dyn_up_tube_MSTR}
    rotation {yaw=20}
   }
   n6 = Get head_r() -> (head_r:double)
   n7 = "Transform (Modify) Bone - Bone: dyn_lw_tube_MSTR"(ComponentPose=@n5, Alpha=@n6, Translation="0,0,0", Rotation="7,-30,0.000", Scale="1,1,1") {
    rotationMode=BMM_Additive
    rotationSpace=BCS_BoneSpace
    scaleMode=BMM_Additive
    scaleSpace=BCS_BoneSpace
    translationMode=BMM_Additive
    translationSpace=BCS_BoneSpace
    alphaBoolBlend {blendOption=Linear}
    alphaScaleBiasClamp {interpSpeedDecreasing=30, interpSpeedIncreasing=30}
    boneToModify {boneName=dyn_lw_tube_MSTR}
    rotation {pitch=7, yaw=-30}
   }
   n8 = "Transform (Modify) Bone - Bone: dyn_lw_tube_02"(ComponentPose=@n7, Alpha=@n6, Translation="0,0,0", Rotation="20,25,9.94", Scale="1,1,1") {
    rotationMode=BMM_Additive
    rotationSpace=BCS_BoneSpace
    scaleMode=BMM_Additive
    scaleSpace=BCS_BoneSpace
    translationMode=BMM_Additive
    translationSpace=BCS_BoneSpace
    alphaBoolBlend {blendOption=Linear}
    alphaScaleBiasClamp {interpSpeedDecreasing=30, interpSpeedIncreasing=30}
    boneToModify {boneName=dyn_lw_tube_02}
    rotation=<rot 20 25 9.94>
   }
   n9 = "Transform (Modify) Bone - Bone: dyn_up_tube_03"(ComponentPose=@n8, Alpha=@n6, Translation="0,0,0", Rotation="-15,0,0", Scale="1,1,1") {
    rotationMode=BMM_Additive
    rotationSpace=BCS_BoneSpace
    scaleMode=BMM_Additive
    scaleSpace=BCS_BoneSpace
    translationMode=BMM_Additive
    translationSpace=BCS_BoneSpace
    alphaBoolBlend {blendOption=Linear}
    alphaScaleBiasClamp {interpSpeedDecreasing=30, interpSpeedIncreasing=30}
    boneToModify {boneName=dyn_up_tube_03}
    rotation {pitch=-15}
   }
   n10 = "Transform (Modify) Bone - Bone: dyn_up_tube_04"(ComponentPose=@n9, Alpha=@n6, Translation="0,0,0", Rotation="0.5,1,-0.5", Scale="1,1,1") {
    rotationMode=BMM_Additive
    rotationSpace=BCS_BoneSpace
    scaleMode=BMM_Additive
    scaleSpace=BCS_BoneSpace
    translationMode=BMM_Additive
    translationSpace=BCS_BoneSpace
    alphaBoolBlend {blendOption=Linear}
    alphaScaleBiasClamp {interpSpeedDecreasing=30, interpSpeedIncreasing=30}
    boneToModify {boneName=dyn_up_tube_04}
    rotation=<rot 0.5 1 -0.5>
   }
   n11 = Get head_up() -> (head_up:double)
   n12 = "Transform (Modify) Bone - Bone: dyn_lw_tube_MSTR"(ComponentPose=@n10, Alpha=@n11, Translation="0,0,0", Rotation="-3,-28,-4", Scale="1,1,1") {
    rotationMode=BMM_Additive
    rotationSpace=BCS_BoneSpace
    scaleMode=BMM_Additive
    scaleSpace=BCS_BoneSpace
    translationMode=BMM_Additive
    translationSpace=BCS_BoneSpace
    alphaBoolBlend {blendOption=Linear}
    alphaScaleBiasClamp {interpSpeedDecreasing=30, interpSpeedIncreasing=30}
    boneToModify {boneName=dyn_lw_tube_MSTR}
    rotation=<rot -3 -28 -4>
   }
   n13 = "Transform (Modify) Bone - Bone: dyn_up_tube_MSTR"(ComponentPose=@n12, Alpha=@n11, Translation="0,0,0", Rotation="17,16,-18", Scale="1,1,1") {
    rotationMode=BMM_Additive
    rotationSpace=BCS_BoneSpace
    scaleMode=BMM_Additive
    scaleSpace=BCS_BoneSpace
    translationMode=BMM_Additive
    translationSpace=BCS_BoneSpace
    alphaBoolBlend {blendOption=Linear}
    alphaScaleBiasClamp {interpSpeedDecreasing=30, interpSpeedIncreasing=30}
    boneToModify {boneName=dyn_up_tube_MSTR}
    rotation=<rot 17 16 -18>
   }
   n14 = "Transform (Modify) Bone - Bone: dyn_up_tube_02"(ComponentPose=@n13, Alpha=@n11, Translation="0,0,0", Rotation="-7,-9,25", Scale="1,1,1") {
    rotationMode=BMM_Additive
    rotationSpace=BCS_BoneSpace
    scaleMode=BMM_Additive
    scaleSpace=BCS_BoneSpace
    translationMode=BMM_Additive
    translationSpace=BCS_BoneSpace
    alphaBoolBlend {blendOption=Linear}
    alphaScaleBiasClamp {interpSpeedDecreasing=30, interpSpeedIncreasing=30}
    boneToModify {boneName=dyn_up_tube_02}
    rotation=<rot -7 -9 25>
   }
   n15 = "Transform (Modify) Bone - Bone: dyn_up_tube_03"(ComponentPose=@n14, Alpha=@n11, Translation="0,0,0", Rotation="8,0.000,0.000", Scale="1,1,1") {
    rotationMode=BMM_Additive
    rotationSpace=BCS_BoneSpace
    scaleMode=BMM_Additive
    scaleSpace=BCS_BoneSpace
    translationMode=BMM_Additive
    translationSpace=BCS_BoneSpace
    alphaBoolBlend {blendOption=Linear}
    boneToModify {boneName=dyn_up_tube_03}
    rotation {pitch=8}
   }
   n16 = "Transform (Modify) Bone - Bone: dyn_up_tube_04"(ComponentPose=@n15, Alpha=@n11, Translation="0,0,0", Rotation="22,0.000,0.000", Scale="1,1,1") {
    rotationMode=BMM_Additive
    rotationSpace=BCS_BoneSpace
    scaleMode=BMM_Additive
    scaleSpace=BCS_BoneSpace
    translationMode=BMM_Additive
    translationSpace=BCS_BoneSpace
    alphaBoolBlend {blendOption=Linear}
    alphaScaleBiasClamp {interpSpeedDecreasing=30, interpSpeedIncreasing=30}
    boneToModify {boneName=dyn_up_tube_04}
    rotation {pitch=22}
   }
   n17 = "Transform (Modify) Bone - Bone: dyn_lw_tube_02"(ComponentPose=@n16, Alpha=@n11, Translation="0,0,0", Rotation="0,0,0", Scale="1,1,1") {
    rotationMode=BMM_Additive
    rotationSpace=BCS_BoneSpace
    scaleMode=BMM_Additive
    scaleSpace=BCS_BoneSpace
    translationMode=BMM_Additive
    translationSpace=BCS_BoneSpace
    alphaBoolBlend {blendOption=Linear}
    alphaScaleBiasClamp {interpSpeedDecreasing=30, interpSpeedIncreasing=30}
    boneToModify {boneName=dyn_lw_tube_02}
   }
   n18 = "Transform (Modify) Bone - Bone: dyn_lw_tube_03"(ComponentPose=@n17, Alpha=@n11, Translation="0,0,0", Rotation="0,6,0.000", Scale="1,1,1") {
    rotationMode=BMM_Additive
    rotationSpace=BCS_BoneSpace
    scaleMode=BMM_Additive
    scaleSpace=BCS_BoneSpace
    translationMode=BMM_Additive
    translationSpace=BCS_BoneSpace
    alphaBoolBlend {blendOption=Linear}
    alphaScaleBiasClamp {interpSpeedDecreasing=30, interpSpeedIncreasing=30}
    boneToModify {boneName=dyn_lw_tube_03}
    rotation {yaw=6}
   }
   n19 = "Transform (Modify) Bone - Bone: dyn_lw_tube_MSTR"(ComponentPose=@n18, Alpha=@n1, Translation="0,0,0", Rotation="-1.5,-7,-2", Scale="1,1,1") {
    rotationMode=BMM_Additive
    rotationSpace=BCS_BoneSpace
    scaleMode=BMM_Additive
    scaleSpace=BCS_BoneSpace
    translationMode=BMM_Additive
    translationSpace=BCS_BoneSpace
    alphaBoolBlend {blendOption=Linear}
    alphaScaleBiasClamp {bInterpResult=true, interpSpeedDecreasing=30, interpSpeedIncreasing=30}
    boneToModify {boneName=dyn_lw_tube_MSTR}
    rotation=<rot -1.5 -7 -2>
   }
   n20 = "Transform (Modify) Bone - Bone: dyn_up_tube_MSTR"(ComponentPose=@n19, Alpha=@n1, Translation="0,0,0", Rotation="8,3,1", Scale="1,1,1") {
    rotationMode=BMM_Additive
    rotationSpace=BCS_BoneSpace
    scaleMode=BMM_Additive
    scaleSpace=BCS_BoneSpace
    translationMode=BMM_Additive
    translationSpace=BCS_BoneSpace
    alphaBoolBlend {blendOption=Linear}
    alphaScaleBiasClamp {bInterpResult=true, interpSpeedDecreasing=30, interpSpeedIncreasing=30}
    boneToModify {boneName=dyn_up_tube_MSTR}
    rotation=<rot 8 3 1>
   }
   Output Pose(Result=Component To Local(ComponentPose=@n20)) {layerGroup=None}
  ]}
 EventGraph {nodes [
   Event BlueprintUpdateAnimation()
   > n1 = Sequence!(execute) -> (then_0, then_1, then_2:Exec)
   n2 = GetOwnerSkeletalMeshComponent()
   n3 = BreakRotator(InRot=BreakTransform(InTransform=GetDeltaTransformFromRefPose(self=@n2, BoneName=head, BaseName=spine_05)).Rotation)
   > Set head_l(head_l:double=MapRangeClamped(Value=@n3, InRangeA=0, InRangeB=60)) -> (Output_Get:double)
   > Set head_r(head_r:double=MapRangeClamped(Value=@n3, InRangeA=0, InRangeB=-40)) -> (Output_Get:double)
   > Set head_r_more(head_r_more:double=MapRangeClamped(Value=@n3, InRangeA=-25, InRangeB=-70)) -> (Output_Get:double)
   > Set head_up(head_up:double=MapRangeClamped(Value=@n3.Yaw, InRangeA=0, InRangeB=-35)) -> (Output_Get:double)
   Set is_ADS(execute=@n1.then_2, is_ADS:double="To Float (Boolean)"(InBool=Get bIsTargeting())) -> (Output_Get:double)
   n4 = BreakTransform(InTransform=GetDeltaTransformFromRefPose(self=GetOwnerSkeletalMeshComponent(), BoneName=clavicle_r, BaseName=spine_05))
   Set right_arm_up(execute=@n1.then_1, right_arm_up:double=MapRangeClamped(Value=BreakRotator(InRot=@n4.Rotation).Pitch, InRangeA=10, InRangeB=40)) -> (Output_Get:double)
   n5 = BreakTransform(InTransform=GetDeltaTransformFromRefPose(self=GetOwnerSkeletalMeshComponent(), BoneName=clavicle_l, BaseName=spine_05))
   > Set left_arm_up(left_arm_up:double=MapRangeClamped(Value=BreakRotator(InRot=@n5.Rotation).Pitch, InRangeA=10, InRangeB=40)) -> (Output_Get:double)
  ]}
]
variables [
 Head_L {type=number}
 Head_R {type=number}
 Head_R_More {type=number}
 Head_Up {type=number}
 Is_ADS {type=number}
 Left_Arm_Up {type=number}
 Right_Arm_Up {type=number}
]
